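; Blockout prop Blueprint that builds a chain of spline meshes along an editable spline, with optional closed loop and start/end cap meshes.
; The construction script hides the template mesh, then adds segments following point scale and roll, with optional tension-based tangent smoothing.
; Segments receive material overrides, optional collision that blocks footstep sound traces, and Index/LastIndex material parameters.
; A custom event snaps spline points to a world-space grid and updates the meshes.
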(event Custom|Snap%20SplinePoints%20to%20Grid
  (bind _spline (Variables|Default|GetSpline))
  (for _index (range (Spline|GetNumberOfSplinePoints _spline))
    (Spline|SetLocationatSplinePoint _spline _index (* (Math|Vector|MakeVector (Math|Conversions|ToFloat(Integer) (Math|Float|Round (.x (/ (Spline|GetLocationatSplinePoint _spline _index "World") (Variables|Default|GetGridSize))))) (Math|Conversions|ToFloat(Integer) (Math|Float|Round (.y (/ (Spline|GetLocationatSplinePoint _spline _index "World") (Variables|Default|GetGridSize))))) (Math|Conversions|ToFloat(Integer) (Math|Float|Round (.z (/ (Spline|GetLocationatSplinePoint _spline _index "World") (Variables|Default|GetGridSize)))))) (Variables|Default|GetGridSize)) "World" true)
    (for _array_element (Variables|Default|GetMeshes)
      (SplineMesh|UpdateMesh _array_element))))

(fn SetSplineMeshCollision
  (params (SplineComponent :type Spline%20Mesh%20Component%20Object%20Reference))
  (if (Variables|Default|GetUseCollision)
    (Collision|SetCollisionEnabled SplineComponent "QueryAndPhysics")
    (Collision|SetCollisionResponsetoChannel SplineComponent "ECC_GameTraceChannel11" "ECR_Block") ; Block foot step traces for sound
    (else
      (Collision|SetCollisionEnabled SplineComponent "NoCollision"))))

(fn SetupSplineMesh
  (params (NewMesh :type Static%20Mesh%20Object%20Reference))
  (bind _returnvalue (AddComponent|NODE_AddSplineMeshComponent-2))
  (Components|StaticMesh|SetStaticMesh _returnvalue NewMesh)
  (SplineMesh|SetForwardAxis _returnvalue (Variables|Default|GetForwardAxis) true)
  (Utilities|FlowControl|Sequence
    (:then_0
      (Utilities|FlowControl|Sequence
        (:then_0
          (for (_array_element _array_index) (Rendering|Material|GetMaterials (Variables|Debug|GetStaticMeshComponent))
            (Utilities|IsValid _array_element
              (:Is%20Valid
                (Rendering|Material|SetMaterial _returnvalue _array_index _array_element)))))
        (:then_1
          (for (_array_element_2 _array_index_2) (Variables|Default|GetOverrideMaterial)
            (Utilities|IsValid _array_element_2
              (:Is%20Valid
                (Rendering|Material|SetMaterial _returnvalue _array_index_2 _array_element_2)))))))
    (:then_1
      (Default|SetSplineMeshCollision _returnvalue)
      (Utilities|Array|Add (Variables|Default|GetMeshes) _returnvalue)
      (return _returnvalue))))

(fn ConstructionScript
  (bind _spline (Variables|Default|GetSpline))
  (bind _returnvalue (Spline|GetNumberOfSplinePoints _spline))
  (bind _returnvalue_2 (- _returnvalue 1))
  (bind (_location _tangent) (Spline|GetLocationandTangentatSplinePoint _spline _returnvalue_2 "Local"))
  (bind (_location_2 _tangent_2) (Spline|GetLocationandTangentatSplinePoint _spline 0 "Local"))
  (bind _staticmeshcomponent (Variables|Debug|GetStaticMeshComponent))
  (bind _spline_2 (Variables|Default|GetSpline))
  (Utilities|FlowControl|Sequence
    (:then_0
      (Collision|SetCollisionResponsetoAllChannels _staticmeshcomponent "ECR_Ignore")
      (Variables|/Script/Engine.PrimitiveComponent|SetbGenerateOverlapEvents false _staticmeshcomponent)
      (Rendering|SetVisibility _staticmeshcomponent false false)
      (Collision|SetCollisionEnabled _staticmeshcomponent "NoCollision"))
    (:then_1
      (Transformation|SetRelativeTransform (Variables|Default|GetSpline) (Math|Transform|MakeTransform "0, 0, 0") false false)
      (bind _staticmeshoverride (Variables|Default|GetStaticMeshOverride)
        (:then
          (Variables|Mesh|SetStaticMesh _staticmeshoverride))
        (:else)))
    (:then_2
      (if (not (Variables|Mesh|GetOverrideStaticMeshMobility))
        (for _array_element (Actor|GetComponentsByClass "/Script/Engine.StaticMeshComponent")
          (Transformation|SetMobility _array_element "Static"))))
    (:then_3
      (for _index (range (Spline|GetNumberOfSplineSegments _spline_2))
        (bind _newparam (Default|SetupSplineMesh (Variables|/Script/Engine.StaticMeshComponent|GetStaticMesh (Variables|Debug|GetStaticMeshComponent))))
        (SplineMesh|SetStartandEnd _newparam (Spline|GetLocationandTangentatSplinePoint _spline_2 _index "Local") (Math|Vector|SelectVector (* (Math|Vector|Normalize (.Tangent (Spline|GetLocationandTangentatSplinePoint _spline_2 _index "Local"))) (* (Math|Vector|VectorLength (- (Spline|GetLocationandTangentatSplinePoint _spline_2 (+ _index 1) "Local") (Spline|GetLocationandTangentatSplinePoint _spline_2 _index "Local"))) (Variables|Default|GetTension))) (.Tangent (Spline|GetLocationandTangentatSplinePoint _spline_2 _index "Local")) (Variables|Default|GetAutoSmoothTangent)) (Spline|GetLocationandTangentatSplinePoint _spline_2 (+ _index 1) "Local") (Math|Vector|SelectVector (* (Math|Vector|Normalize (.Tangent (Spline|GetLocationandTangentatSplinePoint _spline_2 (+ _index 1) "Local"))) (* (Math|Vector|VectorLength (- (Spline|GetLocationandTangentatSplinePoint _spline_2 (+ _index 1) "Local") (Spline|GetLocationandTangentatSplinePoint _spline_2 _index "Local"))) (Variables|Default|GetTension))) (.Tangent (Spline|GetLocationandTangentatSplinePoint _spline_2 (+ _index 1) "Local")) (Variables|Default|GetAutoSmoothTangent)) true :!class /Script/Engine.SplineMeshComponent)
        (SplineMesh|SetStartScale _newparam (Math|Vector2D|MakeVector2D (.x (Spline|GetScaleatSplinePoint (Variables|Default|GetSpline) _index)) (.y (Spline|GetScaleatSplinePoint (Variables|Default|GetSpline) _index))) true :!class /Script/Engine.SplineMeshComponent)
        (SplineMesh|SetEndScale _newparam (Math|Vector2D|MakeVector2D (.x (Spline|GetScaleatSplinePoint (Variables|Default|GetSpline) (+ _index 1))) (.y (Spline|GetScaleatSplinePoint (Variables|Default|GetSpline) (+ _index 1)))) true :!class /Script/Engine.SplineMeshComponent)
        (SplineMesh|SetStartRollRadians _newparam (Math|Trig|DegreesToRadians (Spline|GetRollatSplinePoint _spline_2 _index "Local")) true)
        (SplineMesh|SetEndRollRadians _newparam (Math|Trig|DegreesToRadians (Spline|GetRollatSplinePoint _spline_2 (+ _index 1) "Local")) true))
      (Utilities|FlowControl|Sequence
        (:then_0
          (if (Variables|Default|GetClosedSpline)
            (Spline|SetClosedLoop _spline true true)
            (bind _newparam_2 (Default|SetupSplineMesh (Variables|/Script/Engine.StaticMeshComponent|GetStaticMesh (Variables|Debug|GetStaticMeshComponent))))
            (SplineMesh|SetStartandEnd _newparam_2 _location (Math|Vector|SelectVector (* (Math|Vector|Normalize _tangent) (* (Math|Vector|VectorLength (- _location_2 _location)) (Variables|Default|GetTension))) _tangent (Variables|Default|GetAutoSmoothTangent)) _location_2 (Math|Vector|SelectVector (* (Math|Vector|Normalize _tangent_2) (* (Math|Vector|VectorLength (- _location _location_2)) (Variables|Default|GetTension))) _tangent_2 (Variables|Default|GetAutoSmoothTangent)) true :!class /Script/Engine.SplineMeshComponent)
            (SplineMesh|SetStartScale _newparam_2 (Math|Vector2D|MakeVector2D (.x (Spline|GetScaleatSplinePoint (Variables|Default|GetSpline) _returnvalue_2)) (.y (Spline|GetScaleatSplinePoint (Variables|Default|GetSpline) _returnvalue_2))) true :!class /Script/Engine.SplineMeshComponent)
            (SplineMesh|SetEndScale _newparam_2 (Math|Vector2D|MakeVector2D (.x (Spline|GetScaleatSplinePoint (Variables|Default|GetSpline) 0)) (.y (Spline|GetScaleatSplinePoint (Variables|Default|GetSpline) 0))) true :!class /Script/Engine.SplineMeshComponent)
            (SplineMesh|SetStartRollRadians _newparam_2 (Math|Trig|DegreesToRadians (Spline|GetRollatSplinePoint _spline _returnvalue_2 "Local")) true)
            (SplineMesh|SetEndRollRadians _newparam_2 (Math|Trig|DegreesToRadians (Spline|GetRollatSplinePoint _spline 0 "Local")) true)))
        (:then_1
          (for (_array_element_2 _array_index) (Variables|Default|GetMeshes)
            (Rendering|Material|SetScalarParameterValueonMaterials _array_element_2 "Index" (Math|Conversions|ToFloat(Integer) _array_index))
            (Rendering|Material|SetScalarParameterValueonMaterials _array_element_2 "LastIndex" (- (Math|Conversions|ToFloat(Integer) (Spline|GetNumberOfSplinePoints (Variables|Default|GetSpline))) 1.0))
            (for (_array_element_3 _array_index_2) (Variables|Default|GetOverrideMaterial)
              (Utilities|IsValid _array_element_3
                (:Is%20Valid
                  (Rendering|Material|SetMaterial _array_element_2 _array_index_2 _array_element_3))))))))
    (:then_4
      (Utilities|FlowControl|Sequence
        (:then_0
          (if (Variables|Default|GetEnableStartMesh)
            (bind _startmesh (Variables|Default|GetStartMesh)
              (:then
                (bind _returnvalue_3 (AddComponent|NODE_AddStaticMeshComponent-0 :RelativeTransform (Math|Transform|MakeTransform (+ (.location (Spline|GetTransformatSplinePoint (Variables|Default|GetSpline) 0 "Local" true)) (Math|Vector|RotateVector (.location (Variables|Default|GetStartMeshTransform)) (.rotation (Spline|GetTransformatSplinePoint (Variables|Default|GetSpline) 0 "Local" true)))) (Math|Conversions|ToRotator(Quat) (* (Math|Conversions|ToQuaternion(Rotator) (.rotation (Spline|GetTransformatSplinePoint (Variables|Default|GetSpline) 0 "Local" true))) (Math|Conversions|ToQuaternion(Rotator) (.rotation (Variables|Default|GetStartMeshTransform))))) (* (.scale (Spline|GetTransformatSplinePoint (Variables|Default|GetSpline) 0 "Local" true)) (.scale (Variables|Default|GetStartMeshTransform))))))
                (Components|StaticMesh|SetStaticMesh _returnvalue_3 _startmesh)
                (for (_array_element_4 _array_index_3) (Variables|Default|GetStartMesh%20OverrideMaterial)
                  (Utilities|IsValid _array_element_4
                    (:Is%20Valid
                      (Rendering|Material|SetMaterial _returnvalue_3 _array_index_3 _array_element_4)))))
              (:else))))
        (:then_1
          (if (Variables|Default|GetEnableEndMesh)
            (bind _endmesh (Variables|Default|GetEndMesh)
              (:then
                (bind _returnvalue_4 (AddComponent|NODE_AddStaticMeshComponent-1 :RelativeTransform (Math|Transform|MakeTransform (+ (.location (Spline|GetTransformatSplinePoint (Variables|Default|GetSpline) (- (Spline|GetNumberOfSplinePoints (Variables|Default|GetSpline)) 1) "Local" true)) (Math|Vector|RotateVector (.location (Variables|Default|GetEndMeshTransform)) (.rotation (Spline|GetTransformatSplinePoint (Variables|Default|GetSpline) (- (Spline|GetNumberOfSplinePoints (Variables|Default|GetSpline)) 1) "Local" true)))) (Math|Conversions|ToRotator(Quat) (* (Math|Conversions|ToQuaternion(Rotator) (.rotation (Spline|GetTransformatSplinePoint (Variables|Default|GetSpline) (- (Spline|GetNumberOfSplinePoints (Variables|Default|GetSpline)) 1) "Local" true))) (Math|Conversions|ToQuaternion(Rotator) (.rotation (Variables|Default|GetEndMeshTransform))))) (* (.scale (Spline|GetTransformatSplinePoint (Variables|Default|GetSpline) (- (Spline|GetNumberOfSplinePoints (Variables|Default|GetSpline)) 1) "Local" true)) (.scale (Variables|Default|GetEndMeshTransform))))))
                (Components|StaticMesh|SetStaticMesh _returnvalue_4 _endmesh)
                (for (_array_element_5 _array_index_4) (Variables|Default|GetEndMesh%20OverrideMaterial)
                  (Utilities|IsValid _array_element_5
                    (:Is%20Valid
                      (Rendering|Material|SetMaterial _returnvalue_4 _array_index_4 _array_element_5)))))
              (:else))))))))
